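; Artemis war decal actor that can be tied to a list of parent building walls and customized per placed instance.
; The construction script, on cosmetic-capable instances, selects one of five decal materials and feeds color, normal strength, erosion, gradient and tiling settings into a dynamic material.
; On BeginPlay it prunes invalid walls via reverse iteration and binds to each valid wall's destruction.
; The decal destroys itself when the last tracked wall is destroyed.
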(event EventBeginPlay
  (bind (_arrayindex _arrayelement) (Utilities|Array|ReverseforEachLoop (Variables|Default|GetParentBuildingWalls))
    (:LoopBody
      (Utilities|IsValid _arrayelement
        (:Is%20Valid
          (Game|BindEventtoOnDestroyed _arrayelement (delegate Custom|ParentWallDestroyed%20)))
        (:Is%20Not%20Valid
          (Utilities|Array|RemoveItem (Variables|Default|GetParentBuildingWalls) _arrayelement))))))

(event Custom|ParentWallDestroyed%20
  (params (DestroyedActor :type Actor%20Object%20Reference))
  (bind _parentbuildingwalls (Variables|Default|GetParentBuildingWalls))
  (bind _asparent_building_wall (Utilities|Casting|CastToParent_BuildingWall DestroyedActor))
  (Utilities|Array|RemoveItem _parentbuildingwalls _asparent_building_wall)
  (if (Utilities|Array|IsEmpty _parentbuildingwalls)
    (Actor|DestroyActor)))

(fn ConstructionScript
  (Networking|CanExecuteCosmeticEvents
    (:True
      (bind _output_get (Variables|Default|SetMaterialVariation (Math|Integer|Clamp(Integer) (Variables|Default|GetMaterialVariation) 0 4)))
      (Rendering|Components|Decal|SetDecalMaterial (Variables|Default|GetDecal) (Utilities|Select :Index _output_get :Option%200 "/Game/Athena/Environments/Landscape/Blueprints/WarDecals/M_Artemis_War_Decal.M_Artemis_War_Decal" :Option%201 "/Game/Athena/Environments/Landscape/Blueprints/WarDecals/M_Artemis_War_Decal_B_Inst.M_Artemis_War_Decal_B_Inst" :Option%202 "/Game/Athena/Environments/Landscape/Blueprints/WarDecals/M_Artemis_War_Decal_H_Inst.M_Artemis_War_Decal_H_Inst" :Option%203 "/Game/Athena/Environments/Landscape/Blueprints/WarDecals/M_Artemis_War_Decal_I_Inst.M_Artemis_War_Decal_I_Inst" :Option%204 "/Game/Athena/Environments/Landscape/Blueprints/WarDecals/M_Artemis_War_Decal_Tiling_Inst.M_Artemis_War_Decal_Tiling_Inst"))
      (bind _returnvalue (Rendering|Components|Decal|CreateDynamicMaterialInstance (Variables|Default|GetDecal)))
      (Variables|Default|SetDecalMID _returnvalue)
      (bind _output_get_2 (Variables|Default|SetNormalMapStrength (Math|Float|Clamp(Float) (Variables|Default|GetNormalMapStrength))))
      (Rendering|Material|SetVectorParameterValue (Variables|Default|GetDecalMID) "NormalStrength" (Math|Color|MakeColor _output_get_2 _output_get_2 1.0) :!class /Script/Engine.MaterialInstanceDynamic)
      (bind _output_get_3 (Variables|Default|SetDetailNormalMapStrength (Math|Float|Clamp(Float) (Variables|Default|GetDetailNormalMapStrength))))
      (Rendering|Material|SetVectorParameterValue (Variables|Default|GetDecalMID) "DetailNormalStrength" (Math|Color|MakeColor _output_get_3 _output_get_3) :!class /Script/Engine.MaterialInstanceDynamic)
      (Rendering|Material|SetVectorParameterValue (Variables|Default|GetDecalMID) "WarAlbedoColor" (Variables|Default|GetColorHigh) :!class /Script/Engine.MaterialInstanceDynamic)
      (Rendering|Material|SetVectorParameterValue (Variables|Default|GetDecalMID) "WarAlbedoColorLow" (Variables|Default|GetColorLow) :!class /Script/Engine.MaterialInstanceDynamic)
      (bind _output_get_4 (Variables|Default|SetErosion (Math|Float|Clamp(Float) (Variables|Default|GetErosion))))
      (Rendering|Material|SetScalarParameterValue (Variables|Default|GetDecalMID) "Erosion" _output_get_4 :!class /Script/Engine.MaterialInstanceDynamic)
      (Rendering|Material|SetScalarParameterValue (Variables|Default|GetDecalMID) "SphericalGradientRadius" (Variables|Default|GetSphericalGradientRadius) :!class /Script/Engine.MaterialInstanceDynamic)
      (Rendering|Material|SetScalarParameterValue (Variables|Default|GetDecalMID) "SphericalGradientHardness" (Variables|Default|GetSphericalGradientHardness) :!class /Script/Engine.MaterialInstanceDynamic)
      (Rendering|Material|SetScalarParameterValue (Variables|Default|GetDecalMID) "WorldTilingDivisor" (Variables|Default|GetWorldTilingDivisor) :!class /Script/Engine.MaterialInstanceDynamic))))
