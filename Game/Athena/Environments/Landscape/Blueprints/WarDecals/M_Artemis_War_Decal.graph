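# Deferred decal material for Artemis war decals on landscape environments, with switchable world-aligned or mesh UVs.
# A packed mask texture supplies specular, diffuse, roughness and alpha channels, alongside a two-color albedo lerp.
# The main normal and a detail normal are strength-scaled, added and normalized.
# Opacity multiplies alpha by a spherical UV gradient, subtracts erosion and fades where the decal surface normal aligns with object orientation.
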Material /Game/Athena/Environments/Landscape/Blueprints/WarDecals/M_Artemis_War_Decal
graph {
 properties {blendMode=BLEND_Modulate, materialDomain=MD_DeferredDecal, pixelDepthOffsetMode=PDOM_Legacy, usedShadingModels=Default Lit}
 nodes [
  LinearInterpolate(
   A=VectorParameter() {parameterName=WarAlbedoColorLow, defaultValue=<rgba 0.0573 0.0477 0.0409 0>}.RGB
   B=VectorParameter() {parameterName=WarAlbedoColor, defaultValue=<rgba 0.1476 0.131 0.1191 0>}.RGB
   Alpha=NamedRerouteUsage()
  )
  MaterialFunctionCall() {materialFunction=/Game/Weapons/Prototype/SprayGun/Materials/MF_DecalFacingMask}
  n1 = ObjectOrientation()
  n2 = StaticSwitchParameter(
   True=Divide(
    A=MaterialFunctionCall!(UseLocalSpace, WorldPosition, NormalOverride=@n1) {materialFunction=/Game/Environments/Apollo/Sets/Skyfire/Materials/MaterialFunctions/MF_WorldAlignedUVs_HardEdge}.Result
    B=ScalarParameter() {defaultValue=1200, parameterName=WorldTilingDivisor}
   )
   False=TextureCoordinate()
  ) {parameterName=UseWorldAlignedUVs}
  n3 = TextureSampleParameter2D(UVs=@n2) {
   parameterName=Masks
   samplerType=SAMPLERTYPE_LinearColor
   texture=&/T_Artemis_War_Decal_A
  }
  NamedRerouteDeclaration(None=@n3.R) {name=Specular, nodeColor=<rgba 0.1016 0 1 1>}
  NamedRerouteDeclaration(None=@n3.G) {name=Diffuse, nodeColor=<rgba 0 1 0 1>}
  NamedRerouteDeclaration(None=@n3.B) {name=Roughness, nodeColor=<rgba 0 0 1 1>}
  NamedRerouteDeclaration(None=@n3.A) {name=Alpha, nodeColor=<rgba 1 1 1 1>}
  n4 = TextureSampleParameter2D(UVs=@n2) {
   parameterName=Normal
   samplerType=SAMPLERTYPE_Normal
   texture=&/T_Artemis_War_Decal_A_Normal
  }
  n5 = Multiply(A=@n4.RGB, B=VectorParameter() {parameterName=NormalStrength, defaultValue=<rgba 1 1 1 0>}.RGB)
  n6 = Multiply(
   A=TextureSample() {samplerType=SAMPLERTYPE_Normal, texture=/Game/Athena/Environments/Landscape/Textures/T_Artemis_War_Normal_B}.RGB
   B=VectorParameter() {parameterName=DetailNormalStrength, defaultValue=<rgba 1 1 0 0>}.RGB
  )
  NamedRerouteDeclaration(None=Normalize(VectorInput=Add(A=@n5, B=@n6))) {name=Normal, nodeColor=<rgba 0.5 0.5 1 1>}
  n7 = ScalarParameter() {defaultValue=0.5, parameterName=SphericalGradient_Center_U}
  n8 = ScalarParameter() {defaultValue=0.5, parameterName=SphericalGradient_Center_V}
  n9 = Multiply(
   A=NamedRerouteUsage()
   B=SphereMask(
    A=TextureCoordinate()
    B=AppendVector(A=@n7, B=@n8)
    Radius=ScalarParameter() {defaultValue=0.5, parameterName=SphericalGradientRadius}
    Hardness=ScalarParameter() {defaultValue=0.25, parameterName=SphericalGradientHardness}
   )
  )
  n10 = ScalarParameter() {parameterName=Erosion}
  n11 = Saturate(None=DotProduct(A=MaterialFunctionCall() {materialFunction=/Game/Weapons/Prototype/SprayGun/Materials/MF_DecalNormal}.WorldNormal, B=ObjectOrientation()))
  NamedRerouteDeclaration(None=Multiply(A=Subtract(A=@n9, B=@n10), B=OneMinus(None=@n11))) {name=Opacity, nodeColor=<rgba 1 0.4688 0 1>}
  NamedRerouteUsage()
  NamedRerouteUsage()
  NamedRerouteUsage()
  NamedRerouteUsage()
 ]
}
